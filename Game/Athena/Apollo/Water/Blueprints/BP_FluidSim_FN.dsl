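; Fortnite child of the shallow-water fluid sim that supplies dynamic forces from nearby pawns.
; Every N passes it gathers pawns in grid range, keeping the local player and those at significance LOD 0 or driving.
; It rebuilds the mesh-keyed force map, reusing previous entries or registering new ones.
; Vehicles in water match settings by gameplay tag, while on-foot players and mechs in water bodies pass surface height.
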(event EventBeginPlay
  (|CallParentFunction) ; This node is disabled and will not be called.\nDrag off pins to build functionality.
)

(event Collision|EventActorBeginOverlap
  (params (OtherActor :type Actor%20Object%20Reference)))

(event EventTick
  (params (DeltaSeconds :type Float%20%28single-precision%29))
  (|CallParentFunction DeltaSeconds) ; This node is disabled and will not be called.\nDrag off pins to build functionality.
)

(event EventGetPlayerPawnForces
  (if (Variables|Default|GetUse%20FN%20Pawn%20Forces)
    (CallFunction|GetFortnitePawnForces)))

(fn GetFortnitePawnForces
  (bind _pawn (CallFunction|GetLocalPawn))
  (bind _dynamicforces (Variables|Forces|GetDynamicForces))
  (bind _previousdynamicforces (LocalVariables|GetPreviousDynamicForces))
  (bind _previousdynamicforces_2 (LocalVariables|GetPreviousDynamicForces))
  (bind _vehicletypemap (Variables|Default|GetVehicleTypeMap))
  (if (== (Math|Integer|%25%28Integer%29 (Variables|Default|GetPasses) (Variables|Forces|GetPawn%20Check%20Every%20N%20Frames)) 0)
    (Utilities|Array|Clear (Variables|Default|GetRelevantFortPawns))
    (bind _outfortpawns (Fortnite|GetFortPawnsInRange (Variables|Debug|GetGridCenter) (* (Variables|Simulation|GetFluid%20Size) 0.707)))
    (bind _output_get (LocalVariables|SetNearbyPawns _outfortpawns))
    (Utilities|DebugPrint (Utilities|String|Append (Utilities|String|ToString(Integer) (Utilities|Array|Length _output_get)) "Nearby Pawns from GetFortPawnsInRange")
      (:Out
        (for _array_element _output_get
          ;[1 1 1] Get Pawns with LOD=0
          (bind _asfort_player_pawn (Utilities|Casting|CastToFortPlayerPawn _array_element))
          (if (or (== (LOD|GetSignificanceLOD _asfort_player_pawn) 0) (Vehicle|IsDrivingVehicle _asfort_player_pawn))
            (bind _returnvalue (Water|GetShallowWaterMaxDynamicForces))
            (if (< (Utilities|Array|Length (LocalVariables|GetRelevantPawns)) _returnvalue)
              (Utilities|Array|Add (Variables|Default|GetRelevantFortPawns) _asfort_player_pawn)
              (Utilities|DebugPrint (Utilities|String|Append "Pawn added to list of Relevant Pawns due to being LOD 0 or driving a vehicle.")))
            (else
              (Utilities|DebugPrint (Utilities|String|Append "Pawn Ignored because was either not LOD 0, or was not driving vehicle.")))))
        (Utilities|FlowControl|Sequence
          (:then_0
            ;[1 1 1] Ensure Local Player is included
            (if (Utilities|IsValid _pawn :!class /Script/Engine.KismetSystemLibrary)
              (bind _asfort_player_pawn_2 (Utilities|Casting|CastToFortPlayerPawn _pawn))
              (Utilities|Array|AddUnique (Variables|Default|GetRelevantFortPawns) _asfort_player_pawn_2)))
          (:then_1
            (LocalVariables|SetPreviousDynamicForces _dynamicforces)
            (Utilities|Map|Clear _dynamicforces)
            (for _array_element_2 (Variables|Default|GetRelevantFortPawns)
              (Utilities|IsValid _array_element_2
                (:Is%20Valid
                  (bind _asfort_player_pawn_3 (Utilities|Casting|CastToFortPlayerPawn _array_element_2))
                  (bind _vehicle_occupied (Utilities|IsinVehicle _array_element_2)
                    (:true
                      ;[1 1 1] Boat Mesh Force
                      (if (Class|FortVehicleInterface|IsinWater _vehicle_occupied)
                        (bind _keys (Utilities|Map|Keys _vehicletypemap))
                        (bind _array_element_3 (Utilities|Array|ForEachLoopwithBreak _keys)
                          (:LoopBody
                            (if (and (GameplayTags|HasTag (Class|FortVehicleInterface|GetVehicleTags _vehicle_occupied) _array_element_3) (.ReturnValue (Utilities|Map|Find _vehicletypemap _array_element_3)))
                              (Utilities|FlowControl|Sequence
                                (:then_0
                                  _)
                                (:then_1
                                  _))
                              (if (.ReturnValue (Utilities|Map|Find _previousdynamicforces_2 (Class|FortVehicleInterface|GetSkeletalMeshComponent _vehicle_occupied)))
                                (Utilities|Map|Add (Variables|Forces|GetDynamicForces) (Class|FortVehicleInterface|GetSkeletalMeshComponent _vehicle_occupied) (Utilities|Map|Find _previousdynamicforces_2 (Class|FortVehicleInterface|GetSkeletalMeshComponent _vehicle_occupied))) ; Reuse Force
                                (else
                                  (CallFunction|RegisterDynamicForce (Utilities|Map|Find _vehicletypemap _array_element_3) (Class|FortVehicleInterface|GetSkeletalMeshComponent _vehicle_occupied)))))))))
                    (:false
                      ;[1 1 1] Player Skeletal Mesh Force
                      (Utilities|IsValid (WaterBody|GetCurrentWaterBody _asfort_player_pawn_3)
                        (:Is%20Valid
                          (bind _outwatersurfaceinfo (WaterBody|GetWaterSurfaceInfoStruct (WaterBody|GetCurrentWaterBody _asfort_player_pawn_3) (Transformation|GetWorldLocation (Variables|/Script/Engine.Character|GetMesh _asfort_player_pawn_3)) false false false))
                          (if (.ReturnValue (Utilities|Map|Find _previousdynamicforces (Variables|/Script/Engine.Character|GetMesh _array_element_2)))
                            (bind (_waterplanelocation _waterplanenormal _watersurfacelocation) (Utilities|Struct|BreakWaterSurfaceInfo _outwatersurfaceinfo))
                            (bind (_watersurfacelocation_x _watersurfacelocation_y _watersurfacelocation_z) (Math|Vector|BreakVector _watersurfacelocation))
                            (bind _structout (Utilities|Struct|SetmembersinFluidForceDynamicPerInstanceData (Utilities|Map|Find _previousdynamicforces (Variables|/Script/Engine.Character|GetMesh _array_element_2)) _watersurfacelocation_z))
                            (Utilities|Map|Add (Variables|Forces|GetDynamicForces) (Variables|/Script/Engine.Character|GetMesh _array_element_2) _structout) ; Reuse Force
                            (else
                              (Utilities|Casting|CastToFortCharacterVehicle _asfort_player_pawn_3
                                (:then
                                  (CallFunction|RegisterDynamicForce (Variables|Default|GetMechForceSettings) (Variables|/Script/Engine.Character|GetMesh _array_element_2) _watersurfacelocation_z))
                                (:CastFailed
                                  (CallFunction|RegisterDynamicForce (Variables|Default|GetPlayerForceSettings) (Variables|/Script/Engine.Character|GetMesh _array_element_2) _watersurfacelocation_z))))))))))))))))))

(fn GetLocalPawn
  (bind _returnvalue (Fortnite|GetFirstLocalFortPlayerController true))
  (if (Utilities|IsValid _returnvalue :!class /Script/Engine.KismetSystemLibrary)
    (return (Pawn|GetControlledPawn _returnvalue))
    (else
      (return 0))))

(fn ConstructionScript
  (|CallParentFunction))
