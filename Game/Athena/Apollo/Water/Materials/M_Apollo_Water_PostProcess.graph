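# Post-process material for Apollo water that applies underwater color extinction and fog to the scene color after depth of field.
# Scene depth, corrected for view angle, drives a Custom exp node producing per-channel extinction and a fog blend toward a water color.
# Water height from the velocity-and-height texture, remapped through landscape collection parameters, removes both effects from pixels above the surface.
# A custom depth comparison limits the result to pixels in front of custom depth or where custom depth nears the camera.
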Material /Game/Athena/Apollo/Water/Materials/M_Apollo_Water_PostProcess
graph {
 properties {blendableLocation=BL_SceneColorAfterDOF, materialDomain=MD_PostProcess, pixelDepthOffsetMode=PDOM_Legacy, usedShadingModels=Default Lit}
 nodes [
  n1 = ShadingPathSwitch(Default=ScreenPosition().ViewportUV, Mobile=TextureCoordinate()) {desc=Migrated from FeatureLevelSwitch}
  n2 = ComponentMask(None=SceneTexture(UVs=@n1) {sceneTextureId=PPI_SceneDepth}.Color) {r=true}
  n3 = ComponentMask(None=SceneTexture(UVs=@n1) {sceneTextureId=PPI_CustomDepth}.Color) {r=true}
  n4 = Subtract(A=@n2, B=Multiply(A=@n3, B=ScalarParameter() {defaultValue=1.01, parameterName=Custom Depth Comparison offset}))
  Ceil(None=@n4)
  n5 = ComponentMask(None=SceneTexture(UVs=ScreenPosition().ViewportUV) {sceneTextureId=PPI_PostProcessInput0}.Color) {b=true, g=true, r=true}
  n6 = DotProduct(
   A=CameraVectorWS()
   B=MaterialFunctionCall() {materialFunction=&CameraDirectionVector}.Result
  )
  n7 = Multiply(A=Divide(A=@n2, B=Abs(None=@n6)))
  n8 = VectorParameter() {parameterName=EctinctionColor, defaultValue=<rgba 2 15 75 0>}
  n9 = ScalarParameter() {defaultValue=50, parameterName=ExtinctionDist}
  MaterialExpressionAppendVector_2 = AppendVector(A=Divide(A=@n7, B=Multiply(A=@n8.RGB, B=@n9)), B=Divide(A=@n7, B=ScalarParameter() {defaultValue=5000, parameterName=Fog_Range}))
  n10 = Custom!(x=@MaterialExpressionAppendVector_2) {
   code="exp(-x);"
   desc="RGB = Extinction, A=Fog"
   description=exp
   outputType=CMOT_Float4
   inputs [{inputName=x, input {expression=&self:MaterialExpressionAppendVector_2}}]
  }
  n11 = ComponentMask(None=WorldPosition().XYZ) {b=true}
  n12 = MaterialFunctionCall!(Far Mask, Texture=TextureObjectParameter() {parameterName=WaterVelocityAndHeight, texture=&Black}, Position) {materialFunction=/FortWater/Materials/Functions/WaterBodyVelocityAndHeight}
  n13 = ScalarParameter() {defaultValue=-2, parameterName=asdfasdfasdf}
  n14 = Add(
   A=CollectionParameter() {collection=/Landmass/Landscape/BlueprintBrushes/MPC/MPC_Landscape, parameterName=LandscapeZLocation}
   B=Multiply(
    A=CollectionParameter() {collection=&MPC_Landscape, parameterName=LandscapeZScale}
    B=Add(A=@n12.Water Z Height, B=@n13)
   )
  )
  n15 = Saturate(None=Ceil(None=Subtract(A=@n11, B=@n14)))
  n16 = LinearInterpolate(A=ComponentMask(None=@n10) {b=true, g=true, r=true}, Alpha=@n15) {constA=1}
  n17 = OneMinus(None=Saturate(None=ComponentMask(None=@n10) {a=true}))
  n18 = Saturate(None=Divide(A=@n4, B=ScalarParameter() {defaultValue=1, parameterName=Custom Depth Fade}))
  n19 = ScalarParameter() {defaultValue=10, parameterName=MinD}
  n20 = Saturate(None=Divide(A=Subtract(A=@n3, B=@n19), B=ScalarParameter() {defaultValue=4, parameterName=FadeFade}))
  LinearInterpolate(
   A=@n5
   B=LinearInterpolate(
    A=Multiply(A=@n5, B=@n16)
    B=VectorParameter() {parameterName=WaterFogColor_1, defaultValue=<rgba 0.0212 0.0682 0.1 0>}.RGB
    Alpha=LinearInterpolate(A=@n17, Alpha=@n15) {constB=0}
   )
   Alpha=Max(A=OneMinus(None=@n18), B=OneMinus(None=@n20))
  )
  Min(A=@n2, B=ComponentMask(None=SceneTexture(UVs=@n1) {sceneTextureId=PPI_CustomDepth}.Color) {r=true})
  PixelDepth()
 ]
}
